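# GnatGala sound cue for a looping proximity threat layer, routed to the Fortnite boss fight music sound class.
# An Enveloper at the root fades volume in from silence over the first three seconds, feeding an Attenuation node and then a looping wave player with the distant classic loop.
# Distant-loop attenuation shapes falloff, and a concurrency override limits the cue to a single playing instance.
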SoundCue /FNM_AI_GnatGala/Audio/IP_SC_GnatGala_Proximity_Threat_Loop
root=Enveloper_0
nodes [
 {id=Attenuation_0, children [WavePlayer_0], properties {attenuationSettings=&/ATT_GnatGala_Proximity_Distant_Loop}}
 {
  id=Enveloper_0
  children [Attenuation_0]
  properties {pitchCurve {editorCurveData {keys [{time=0}]}}, volumeCurve {editorCurveData {keys [<at 0 0>,<at 3 1>]}}}
 }
 {id=WavePlayer_0, properties {bLooping=true, soundWaveAssetPtr=&/IP_GnatGala_Proximity_Distant_Classic_Loop}}
]
properties {
 attenuationSettings=&ATT_GnatGala_Proximity_Distant_Loop
 bOverrideConcurrency=true
 maxDistance=8000
 soundClassObject=/Game/Sounds/FortSoundClassesAndModes/Fort_BossFight_SoundClass/Fort_Music_Boss_Fight
 volumeMultiplier=1.5
 assetUserData [&self:AudioAssetUserData_0]
 concurrencyOverrides {maxCount=1}
}
