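# Sound cue for the GnatGala melee attack hitting the world, randomly picking one of four impact waves with slight pitch variation.
# An islocal branch plays the local sound at full volume without an attenuation node.
# Otherwise an isenemy branch applies pickaxe impact attenuation, cutting non-enemy hits to a fifth of the volume while enemy or unset stays full.
# Concurrency is limited per owner, and the cue primes on load.
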SoundCue /FNM_AI_GnatGala/Audio/SC_GnatGala_Attack_Melee_Impact_World
root=Branch_0
nodes [
 {id=Attenuation_0, children [Modulator_4], properties {attenuationSettings=/Game/Sounds/Attenuation/Fort_Weapon_Pickaxe_Player_Impact_Att}}
 {id=Attenuation_2, children [Modulator_3], properties {attenuationSettings=&Fort_Weapon_Pickaxe_Player_Impact_Att}}
 {id=Branch_0, children [Modulator_2,Branch_3,Branch_3], properties {boolParameterName=islocal}}
 {id=Branch_3, children [Attenuation_0, Attenuation_2, Attenuation_0], properties {boolParameterName=isenemy}}
 {id=Modulator_2, children [Modulator_6], properties {pitchMax=1, pitchMin=1, volumeMax=1, volumeMin=1}}
 {id=Modulator_3, children [Modulator_6], properties {pitchMax=1, pitchMin=1, volumeMax=0.2, volumeMin=0.2}}
 {id=Modulator_4, children [Modulator_6], properties {pitchMax=1, pitchMin=1, volumeMax=1, volumeMin=1}}
 {id=Modulator_6, children [Random_2], properties {pitchMax=1.1, pitchMin=1, volumeMax=1, volumeMin=1}}
 {
  id=Random_2
  children [WavePlayer_7, WavePlayer_8, WavePlayer_9, WavePlayer_10]
  properties {bShouldExcludeFromBranchCulling=true, weights [1,1,1,1]}
 }
 {id=WavePlayer_10, properties {soundWaveAssetPtr=&/GnatGala_Melee_Impact_World_04}}
 {id=WavePlayer_7, properties {soundWaveAssetPtr=&/GnatGala_Melee_Impact_World_01}}
 {id=WavePlayer_8, properties {soundWaveAssetPtr=&/GnatGala_Melee_Impact_World_02}}
 {id=WavePlayer_9, properties {soundWaveAssetPtr=&/GnatGala_Melee_Impact_World_03}}
]
properties {
 attenuationSettings=/Game/Sounds/Attenuation/Fort_Pickaxe_Impact_Att
 bOverrideConcurrency=true
 bPrimeOnLoad=true
 maxDistance=4300
 soundClassObject=&Fort_Impacts
 volumeMultiplier=1
 attenuationOverrides {
  bAttenuateWithLPF=true
  distanceAlgorithm=Logarithmic
  falloffDistance=800
  lPFRadiusMax=800
  lPFRadiusMin=400
  attenuationShapeExtents {x=300}
 }
 concurrencyOverrides {bLimitToOwner=true, maxCount=3}
 concurrencySet [&Player_Melee_Weapon_Conc]
}
