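# Sound cue for the GnatGala AI's heavy melee swing attack.
# A random node picks one of five equally weighted swing wave variations and is excluded from branch culling.
# It uses the melee weapon sound class with a shared close-range stereo melee attenuation, reduced volume, and a concurrency override that caps simultaneous instances at a low count.
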SoundCue /FNM_AI_GnatGala/Audio/SC_GnatGala_Attack_Melee_Swing_Heavy
root=Random_0
nodes [
 {
  id=Random_0
  children [WavePlayer_3, WavePlayer_4, WavePlayer_5, WavePlayer_6, WavePlayer_10]
  properties {bShouldExcludeFromBranchCulling=true, weights [1,1,1,1,1]}
 }
 {id=WavePlayer_10, properties {soundWaveAssetPtr=&/GnatGala_Melee_Swing_Heavy_05}}
 {id=WavePlayer_3, properties {soundWaveAssetPtr=&/GnatGala_Melee_Swing_Heavy_01}}
 {id=WavePlayer_4, properties {soundWaveAssetPtr=&/GnatGala_Melee_Swing_Heavy_02}}
 {id=WavePlayer_5, properties {soundWaveAssetPtr=&/GnatGala_Melee_Swing_Heavy_03}}
 {id=WavePlayer_6, properties {soundWaveAssetPtr=&/GnatGala_Melee_Swing_Heavy_04}}
]
properties {
 attenuationSettings=/Game/Sounds/Attenuation/Fort_Weapon_Melee_Attack_Close_Range_Stereo_Att
 bOverrideConcurrency=true
 maxDistance=10058.75
 soundClassObject=&Fort_WeaponMelee
 volumeMultiplier=0.6
 concurrencyOverrides {maxCount=2}
}
