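# Sound cue for the light melee swing of the GnatGala AI's attack.
# A Random node picks one of six equally weighted swing wave variations, with the node excluded from branch culling.
# Playback uses the Fortnite close-range stereo melee attenuation and the weapon melee sound class at reduced volume.
# A concurrency override caps simultaneous instances at two.
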SoundCue /FNM_AI_GnatGala/Audio/SC_GnatGala_Attack_Melee_Swing_Light
root=Random_1
nodes [
 {
  id=Random_1
  children [WavePlayer_0,WavePlayer_1,WavePlayer_2,WavePlayer_7,WavePlayer_8,WavePlayer_9]
  properties {bShouldExcludeFromBranchCulling=true, weights [1,1,1,1,1,1]}
 }
 {id=WavePlayer_0, properties {soundWaveAssetPtr=&/GnatGala_Melee_Swing_01}}
 {id=WavePlayer_1, properties {soundWaveAssetPtr=&/GnatGala_Melee_Swing_02}}
 {id=WavePlayer_2, properties {soundWaveAssetPtr=&/GnatGala_Melee_Swing_03}}
 {id=WavePlayer_7, properties {soundWaveAssetPtr=&/GnatGala_Melee_Swing_04}}
 {id=WavePlayer_8, properties {soundWaveAssetPtr=&/GnatGala_Melee_Swing_05}}
 {id=WavePlayer_9, properties {soundWaveAssetPtr=&/GnatGala_Melee_Swing_06}}
]
properties {
 attenuationSettings=/Game/Sounds/Attenuation/Fort_Weapon_Melee_Attack_Close_Range_Stereo_Att
 bOverrideConcurrency=true
 maxDistance=10058.75
 soundClassObject=&Fort_WeaponMelee
 volumeMultiplier=0.7
 concurrencyOverrides {maxCount=2}
}
