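# Sound cue for the GnatGala AI's melee attack swing "shing" sound.
# A single wave player feeds a modulator that randomizes pitch around normal on each play.
# It uses close-range stereo melee weapon attenuation and is routed to the boss fight music sound class.
# A concurrency override caps it at two simultaneous instances, stopping the oldest first.
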SoundCue /FNM_AI_GnatGala/Audio/SC_GnatGala_Attack_Melee_Swing_Shing
root=Modulator_0
nodes [
 {id=Modulator_0, children [WavePlayer_0], properties {pitchMax=1.2, pitchMin=0.8}}
 {id=WavePlayer_0, properties {soundWaveAssetPtr=&/GnatGala_Melee_Swing_Shing_01}}
]
properties {
 attenuationSettings=/Game/Sounds/Attenuation/Fort_Weapon_Melee_Attack_Close_Range_Stereo_Att
 bOverrideConcurrency=true
 maxDistance=10058.75
 soundClassObject=/Game/Sounds/FortSoundClassesAndModes/Fort_BossFight_SoundClass/Fort_Music_Boss_Fight
 volumeMultiplier=1
 concurrencyOverrides {maxCount=2, resolutionRule=StopOldest}
}
