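# Looping 2D marked-target SoundCue for GnatGala, judging by its name, with an empty node graph and no root, so it references no audio.
# It routes to the boss-fight music sound class and uses a distant proximity loop attenuation asset.
# An overridden concurrency limits it to a single instance at a time.
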SoundCue /FNM_AI_GnatGala/Audio/SC_GnatGala_MarkedTarget_2D_Loop
root=null
nodes []
properties {
 attenuationSettings=&/ATT_GnatGala_Proximity_Distant_Loop
 bOverrideConcurrency=true
 maxDistance=20000
 soundClassObject=/Game/Sounds/FortSoundClassesAndModes/Fort_BossFight_SoundClass/Fort_Music_Boss_Fight
 volumeMultiplier=1
 concurrencyOverrides {maxCount=1}
}
