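# Looping Sound Cue that plays the proximity threat audio for GnatGala AI, routed to the boss-fight music sound class.
# Distant, close and heartbeat loops each pass through their own attenuation into a weighted Mixer, so distance shapes the layering.
# The heartbeat sits behind a Quality Level node with only its first slot connected.
# An Enveloper fades the mix in over three seconds, and a concurrency override allows one instance at a time.
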SoundCue /FNM_AI_GnatGala/Audio/SC_GnatGala_Proximity_Threat_Loop
root=Enveloper_0
nodes [
 {id=Attenuation_0, children [WavePlayer_0], properties {attenuationSettings=&/ATT_GnatGala_Proximity_Distant_Loop}}
 {id=Attenuation_1, children [WavePlayer_1], properties {attenuationSettings=&/ATT_GnatGala_Proximity_Close_Loop}}
 {id=Attenuation_2, children [WavePlayer_2], properties {attenuationSettings=&/ATT_GnatGala_Proximity_Heartbeat_Loop}}
 {
  id=Enveloper_0
  children [Mixer_0]
  properties {pitchCurve {editorCurveData {keys [{time=0}]}}, volumeCurve {editorCurveData {keys [<at 0 0>,<at 3 1>]}}}
 }
 {id=Mixer_0, children [Attenuation_0, Attenuation_1, QualityLevel_0], properties {inputVolume [1.5,2,4]}}
 {id=QualityLevel_0, children [Attenuation_2, null, null, null]}
 {id=WavePlayer_0, properties {bLooping=true, soundWaveAssetPtr=&/GnatGala_Proximity_Distant_Loop}}
 {id=WavePlayer_1, properties {bLooping=true, soundWaveAssetPtr=&/GnatGala_Proximity_Close_Loop_v2}}
 {id=WavePlayer_2, properties {bLooping=true, soundWaveAssetPtr=&/GnatGala_Proximity_Heartbeat}}
]
properties {
 attenuationSettings=&ATT_GnatGala_Proximity_Distant_Loop
 bOverrideConcurrency=true
 maxDistance=8000
 soundClassObject=/Game/Sounds/FortSoundClassesAndModes/Fort_BossFight_SoundClass/Fort_Music_Boss_Fight
 volumeMultiplier=1
 concurrencyOverrides {maxCount=1}
}
