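# SoundCue for the GnatGala AI's warp, playing a jump-scare stinger.
# A Random node picks one of four equally weighted stinger waves.
# Warp-specific attenuation shapes its spatial falloff, and the boss fight music sound class routes it with a boosted volume.
# A concurrency override limits the cue to a single playing instance at a time.
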SoundCue /FNM_AI_GnatGala/Audio/SC_GnatGala_Warp
root=Random_0
nodes [
 {id=Random_0, children [WavePlayer_3,WavePlayer_4,WavePlayer_5,WavePlayer_6], properties {weights [1,1,1,1]}}
 {id=WavePlayer_3, properties {soundWaveAssetPtr=&/GnatGala_Stinger_JumpScare_01}}
 {id=WavePlayer_4, properties {soundWaveAssetPtr=&/GnatGala_Stinger_JumpScare_02}}
 {id=WavePlayer_5, properties {soundWaveAssetPtr=&/GnatGala_Stinger_JumpScare_03}}
 {id=WavePlayer_6, properties {soundWaveAssetPtr=&/GnatGala_Stinger_JumpScare_04}}
]
properties {
 attenuationSettings=&/ATT_GnatGala_Warp
 bOverrideConcurrency=true
 maxDistance=10000
 soundClassObject=/Game/Sounds/FortSoundClassesAndModes/Fort_BossFight_SoundClass/Fort_Music_Boss_Fight
 volumeMultiplier=1.5
 concurrencyOverrides {maxCount=1}
}
